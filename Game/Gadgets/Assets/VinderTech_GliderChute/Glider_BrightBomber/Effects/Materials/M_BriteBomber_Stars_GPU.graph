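# Masked, lit GPU particle material for the Bright Bomber glider's star effect.
# Random sub-UV cells from star mask and normal sheets combine with particle color and a random blend between ColorA and hue-shifted ColorB.
# Random brightness and a normal-map dot product shade the emissive, while a dimmed copy supplies base color.
# Near-camera fade scales a particle-size offset, and shading-path switches give mobile simpler SubUV animation, color and brightness without that offset.
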Material /Game/Gadgets/Assets/VinderTech_GliderChute/Glider_BrightBomber/Effects/Materials/M_BriteBomber_Stars_GPU
graph {
 properties {
  bUsedWithNiagaraSprites=true
  bUsedWithParticleSprites=true
  blendMode=BLEND_Masked
  pixelDepthOffsetMode=PDOM_Legacy
  usedShadingModels=Default Lit
 }
 nodes [
  n1 = ParticleRandom()
  n2 = VectorParameter() {parameterName=ColorB, defaultValue=<rgba 0 1 1 1>}
  n3 = Constant()
  n4 = ParticleColor()
  n5 = Constant() {r=4}
  n6 = Add(A=ParticleRandom()) {constB=0.05}
  n7 = Time()
  n8 = MaterialFunctionCall!(UseTexCoord 1, Rows=@n5, Columns=@n5, Particle Random Value=Add(A=Frac(None=@n6), B=Multiply(A=@n7))) {materialFunction=/Game/Effects/Fort_Effects/Materials/_Functions/MF_RandomSubUV_GPU_Sprites}
  n9 = TextureSample(UVs=@n8.Result) {samplerType=SAMPLERTYPE_Grayscale, texture=/Game/Effects/Fort_Effects/Textures/T_Stars_Mask}
  n10 = Constant() {r=4}
  n11 = Constant() {r=2}
  n12 = ComponentMask(None=MaterialFunctionCall!(
    U_Images=@n10
    V_Images=@n10
    SubUV Texture RGB=TextureObject() {samplerType=SAMPLERTYPE_Grayscale, texture=&T_Stars_Mask}
    Time Speed=@n11
    Manual Time Control
    Time
    UVs
   ) {materialFunction=/Game/Effects/Fort_Effects/Materials/_Functions/MF_SubUV}.RGB) {g=true}
  n13 = Multiply(
   A=LinearInterpolate(
    A=VectorParameter() {parameterName=ColorA, defaultValue=<rgba 1 0 1 1>}.RGB
    B=MaterialFunctionCall(Hue Shift Percentage=@n1, Texture=@n2.RGB) {materialFunction=&HueShift}.Result
    Alpha=ShadingPathSwitch(Default=@n1, Mobile=@n3) {desc=Migrated from FeatureLevelSwitch}
   )
   B=ShadingPathSwitch(Default=Multiply(A=@n4.RGB, B=@n9.G), Mobile=Multiply(A=@n12, B=@n4.RGB)) {desc=Migrated from FeatureLevelSwitch}
  )
  n14 = Multiply(
   A=@n13
   B=LinearInterpolate(
    A=ScalarParameter() {defaultValue=1, parameterName=BrightLow}
    B=ScalarParameter() {defaultValue=8, parameterName=BrightHigh}
    Alpha=ParticleRandom()
   )
  )
  n15 = ShadingPathSwitch(
   Default=TextureSample(UVs=@n8.Result) {samplerType=SAMPLERTYPE_Normal, texture=/Game/Effects/Fort_Effects/Textures/T_Stars_NM}.RGB
   Mobile=MaterialFunctionCall(
    U_Images=@n10
    V_Images=@n10
    SubUV Texture RGB=TextureObject() {samplerType=SAMPLERTYPE_Normal, texture=&T_Stars_NM}
    Time=@n11
   ) {materialFunction=&MF_SubUV}.RGB
  ) {desc=Migrated from FeatureLevelSwitch}
  n16 = Constant3Vector() {constant=<rgba 1 0 0 0>}
  MaterialFunctionCall(Emissive=Multiply(A=ShadingPathSwitch(Default=@n14, Mobile=@n13) {desc=Migrated from FeatureLevelSwitch}, B=DotProduct(A=@n15, B=@n16))) {materialFunction=&MF_FX_EmissiveAdjustments}
  Multiply(A=ScalarParameter() {defaultValue=0.2, parameterName=BaseColorStrength}, B=@n13)
  ScalarParameter() {parameterName=Metallic}
  ScalarParameter() {defaultValue=1, parameterName=Specular}
  ScalarParameter() {defaultValue=0.25, parameterName=Roughness}
  n17 = MaterialFunctionCall(Distance=ScalarParameter() {defaultValue=216.0992, parameterName=CamFadeDist}) {materialFunction=&Near_Camera_Fade}
  n18 = Transform(None=AppendVector(A=TextureCoordinate(), B=Constant()))
  n19 = ComponentMask(None=ParticleSize()) {r=true}
  ShadingPathSwitch(Default=Multiply(A=LinearInterpolate(Alpha=@n17.Result) {constA=-1, constB=0}, B=Multiply(A=@n18, B=@n19)), Mobile=Constant()) {desc=Migrated from FeatureLevelSwitch}
 ]
}
